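; Actor component that toggles a discoverable gag's sparkle and material visuals on its owner.
; On playlist ready, remote clients check enable hotfixes and, if required, local players' quest tag, hiding visuals locally when that tag is removed.
; Server Begin/End events set replicated bDiscoverableEnded, whose OnRep updates state without rechecking the tag.
; State changes fire begin/ended dispatchers, toggle optionally tag-filtered Niagara components and set Gag_Enabled on dynamic mesh materials.
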(event EventBeginPlay
  ;[0 0.4948 0.004556] BeginPlay
  ; Set the initial state of the DiscoverableGag visuals based on the actor state and enabled variables
  (|CallParentFunction)
  (Class|FortAsyncActionCurrentPlaylistReady|PlaylistDataReadyAsync
    (:OnReady
      (Utilities|IsValid (Components|GetOwner)
        (:Is%20Valid
          (Networking|SwitchHasAuthority
            (:Remote
              (if (and (Variables|DiscoverableGag|GetbEnableDiscoverableGag) (ScalableFloat|GetValueasBool (Variables|DiscoverableGag|Hotfixes|GetRow_GlobalEnableDiscoverableGag)))
                (if (ScalableFloat|GetValueasBool (Variables|DiscoverableGag|Hotfixes|GetRow_GlobalRequireQuestTag))
                  (CallFunction|SetGagStatusbyLocalPlayerRequiredQuest)
                  (else
                    (CallFunction|SetDiscoverableGagState (not (Variables|DiscoverableGag|States|GetbDiscoverableEnded)))))
                (else
                  (CallFunction|SetDiscoverableGagState false) ; Discoverable Gag visuals are disabled 
                )))))))))

(event Custom|SetDiscoverableGagState
  (params (SetDiscoverableGagEnabled :type Boolean))
  ;[0.1829 0.3394 0.8802] SetDiscoverableGagState
  ; Use the current Gag State bool to set all relevant updates in the owning actor
  (bind _output_get (Variables|DiscoverableGag|States|SetbDiscoverableGagIsVisible SetDiscoverableGagEnabled))
  (if _output_get
    (Events|CallDiscoverableGagonBegin)
    (else
      (Events|CallDiscoverableGagonEnded)))
  ;;[0.4423 0.05575 1] Set the visuals states of the owning actor components that distinguish the Discoverable Gag
  (Networking|CanExecuteCosmeticEvents
    (:True
      (CallFunction|SetGagMaterialEffect)
      (CallFunction|SetGagNiagaraEffect))))

(event Custom|SetGagStatusByLocalPlayerRequiredQuest
  (bind _bdiscoverableended (Variables|DiscoverableGag|States|GetbDiscoverableEnded))
  (bind _returnvalue (not _bdiscoverableended))
  (bind _questtagtocheck (Variables|DiscoverableGag|GetQuestTagToCheck))
  ;[0.2344 0.4025 1] SetGagStatusByLocalPlayerRequiredQuest
  ; Check the local players for the required quest tag and update the visuals accordingly
  (Networking|CanExecuteCosmeticEvents
    (:True
      (for _array_element (Fortnite|GetLocalFortPlayerControllers true)
        (if (GameplayTags|HasTag (GameplayTags|GetOwnedGameplayTags (Ability|GetAbilitySystemComponent _array_element) :!class /Script/GameplayTags.BlueprintGameplayTagLibrary) _questtagtocheck)
          (CallFunction|SetDiscoverableGagState _returnvalue) ; local player has the required tag and visuals should be enabled\nunless it has already Ended
          (if _returnvalue
            (bind _asynctaskproxy (Ability|Tasks|WaitGameplayTagRemoveFromActor _array_element _questtagtocheck true)
              (:then
                (Variables|Default|SetTask_QuestTagRemovedAsync _asynctaskproxy))
              (:Removed
                (CallFunction|SetDiscoverableGagState false) ; end the gag - since this is run on the client for local clients only,\nwill allow other players to still see the effect\nTo disable entirely, use the EndDiscoverableGag_Server event
                (Ability|Async|EndAction (Variables|Default|GetTask_QuestTagRemovedAsync)))))
          (else
            (CallFunction|SetDiscoverableGagState false) ; Discoverable Gag visuals are disabled OR\nlocal player does not have the required tag 
          ))))))

(event Custom|SetGagNiagaraEffect
  ;[0.4423 0.05575 1] SetGagNiagaraEffect
  ; Enables/Disables the Niagara sparkle effect - allows the Niagara systems to use a component tag in case the owning actor implements multiple Niagara components and we don't want to disable all of them
  (for _array_element (Actor|GetComponentsByClass (Components|GetOwner) "/Script/Niagara.NiagaraComponent")
    ;;[0.1408 0.3292 0.9323 0.5] Check if the gag actor is using the "RequireGagSparklesTag" which is useful in case it has a niagara component that we DON'T want to disable
    (if (or (not (Variables|DiscoverableGag|GetbRequireGagSparklesComponentTag)) (and (Variables|DiscoverableGag|GetbRequireGagSparklesComponentTag) (Components|ComponentHasTag _array_element (Variables|DiscoverableGag|GetDiscoverableGagComponentTag))))
      (Development|SetHiddeninGame _array_element (not (Variables|DiscoverableGag|States|GetbDiscoverableGagIsVisible)) false))))

(event Custom|SetGagMaterialEffect
  ;[0.4423 0.05575 1] SetGagMaterialEffect
  ; Enables/Disables the Material effect on the static mesh of the owning actor
  (for _array_element (Actor|GetComponentsByClass (Components|GetOwner) "/Script/Engine.StaticMeshComponent")
    (Utilities|IsValid (Rendering|Material|GetMaterial _array_element 0)
      (:Is%20Valid
        (bind _returnvalue (Rendering|Material|CreateDynamicMaterialInstance _array_element 0 (Rendering|Material|GetMaterial _array_element 0) "None" :!class /Script/Engine.PrimitiveComponent))
        (Rendering|Material|SetScalarParameterValue _returnvalue "Gag_Enabled" (Math|Float|SelectFloat 1.0 0.0 (Variables|DiscoverableGag|States|GetbDiscoverableGagIsVisible)) :!class /Script/Engine.MaterialInstanceDynamic) ; set the material effect parameter enabled or disabled
      ))))

(event Custom|EndDiscoverableGag_Server
  ;[0.3516 1 0.5137] BeginDiscoverableGag_Server
  ; Called  from actors that implement this interface when they want to DISABLE the discoverable visual features
  (Utilities|SwitchHasAuthority
    (:Authority
      (Variables|DiscoverableGag|States|SetbDiscoverableEnded true))))

(event Custom|BeginDiscoverableGag_Server
  ;[0.3516 1 0.5137] BeginDiscoverableGag_Server
  ; Called  from actors that implement this interface when they want to re-enable the discoverable visual features
  (Utilities|SwitchHasAuthority
    (:Authority
      (Variables|DiscoverableGag|States|SetbDiscoverableEnded false))))

(fn GetDiscoverableGagEnded
  ;[0.2727 0.4328 0.6302] GetDiscoverableGagEnded
  ; Getter function for private variable "bDiscoverableEnded"
  (return (Variables|DiscoverableGag|States|GetbDiscoverableEnded)))

(fn OnRep_bDiscoverableEnded
  ;[0.3424 0.5068 1] OnRep_DiscoverableEnded
  ; Handles the replicated variable when the Discoverable state is set by the owning actor on the server
  (CallFunction|SetDiscoverableGagState (not (Variables|DiscoverableGag|States|GetbDiscoverableEnded))))
